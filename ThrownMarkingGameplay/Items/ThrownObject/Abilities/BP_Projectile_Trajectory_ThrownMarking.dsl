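; Trajectory preview for the thrown marking item, extending a parent trajectory Blueprint with feedback for the deployed limit.
; Construction creates an at-limit dynamic material from the grenade trajectory line material.
; BeginPlay hides the target and, when a console variable is set, hides the actor from non-instigator local controllers for splitscreen.
; UpdateVisiblityAtLimit swaps target and spline mesh materials by limit state, or applies the invalid material to splines for invalid trajectories.
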(event EventBeginPlay
  (|CallParentFunction)
  (Utilities|FlowControl|Sequence
    (:then_0
      (Rendering|SetVisibility (Variables|Default|GetTarget) false false))
    (:then_1
      (bind _returnvalue (Development|GetConsoleVariableBoolValue "Fort.Rollback.HideWeakSpotsOnNonInstigatorLocalControllers"))
      (if _returnvalue
        ;[1 1 1] Splitscreen fix: Hide projectile trajectory if this is NOT the instigator
        (Fortnite|SetActorHiddenforNonInstigatorLocalPlayerControllers self)))))

(fn UpdateVisiblityAtLimit
  (params (AtLimit :type Boolean))
  (bind _atlimit (Variables|GetAtLimit))
  ; Handle changing the color of the trajectory based on if we are at the deployed limit or not.
  (if (Variables|Default|GetbIsTrajectoryValid)
    (Rendering|Material|SetMaterial (Variables|Default|GetTarget) 0 (select _atlimit (Variables|Default|GetAtLimitTargetMaterial) (Variables|Default|GetDefaultValidTargetMaterial)))
    (for _array_element (Variables|Trajectory|GetSplineMeshComponents)
      (Rendering|Material|SetMaterial _array_element 0 (Selection|SelectMaterialInstance (Variables|Default|GetAt%20Limit%20MID) (Variables|VFX|GetSplineMID) _atlimit)))
    (else
      (for _array_element_2 (Variables|Trajectory|GetSplineMeshComponents)
        (Rendering|Material|SetMaterial _array_element_2 0 (Variables|VFX|GetInvalidSplineMID))))))

(fn ConstructionScript
  (|CallParentFunction)
  (bind _returnvalue (Rendering|Material|CreateDynamicMaterialInstance "/Game/Athena/Items/HeldObjects/M_Grenade_TrajectoryLine_01.M_Grenade_TrajectoryLine_01" "None" "None" :!class /Script/Engine.KismetMaterialLibrary))
  (Variables|Default|SetAt%20Limit%20MID _returnvalue))
